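; Dance-off manager Blueprint for the LootCannon gameplay, providing a check for whether players on both sides are emoting at once.
; IsPlayerOnEachSideEmoting scans the left-side and then right-side player arrays, casting each pawn's controller to the Athena type to query IsPlayingEmote.
; Each loop breaks at the first emoting player and sets a local flag, and the function returns the AND of both flags.
; The construction script is empty.
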(fn IsPlayerOnEachSideEmoting
  (bind _array_element (Utilities|Array|ForEachLoopwithBreak (Variables|Default|GetLeftSidePlayers) :!label _label)
    (:LoopBody
      (bind _asfort_player_pawn_athena (Utilities|Casting|CastToFortPlayerPawnAthena _array_element))
      (bind _asfort_player_controller_athena (Utilities|Casting|CastToFortPlayerControllerAthena (Pawn|GetController _asfort_player_pawn_athena)))
      (if (Emote|IsPlayingEmote _asfort_player_controller_athena)
        (LocalVariables|SetIsLeftEmotingLocal true)
        (goto _label Break)))
    (:Completed
      (bind _array_element_2 (Utilities|Array|ForEachLoopwithBreak (Variables|Default|GetRightSidePlayers) :!label _label_2)
        (:LoopBody
          (bind _asfort_player_pawn_athena_2 (Utilities|Casting|CastToFortPlayerPawnAthena _array_element_2))
          (bind _asfort_player_controller_athena_2 (Utilities|Casting|CastToFortPlayerControllerAthena (Pawn|GetController _asfort_player_pawn_athena_2)))
          (if (Emote|IsPlayingEmote _asfort_player_controller_athena_2)
            (LocalVariables|SetIsRightEmotingLocal true)
            (goto _label_2 Break)))
        (:Completed
          (return (and (LocalVariables|GetIsLeftEmotingLocal) (LocalVariables|GetIsRightEmotingLocal))))))))

(fn ConstructionScript)
