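; Single-use Loot Cannon building actor: interacting lights its fuse, then it fires tiered loot and can launch players standing in front.
; FireLootCannon waits out the fuse, then the server tosses pickups with random spread while clients play effects, audio and camera shake.
; ScalableFloats tune launch and toss behavior; non-interactable cannons become invulnerable.
; Destruction before firing drops fallback loot, and a RepNotify shuts off idle cues after use.
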(event EventBeginPlay
  ;[1 1 1] Init
  (Variables|Default|SetbLaunchPlayerEnabled (ScalableFloat|GetValueasBool (Variables|Default|GetShouldLaunchPlayer)))
  (Variables|Default|SetTossSpreadMultiplier (* (ScalableFloat|GetValueatLevel (Variables|Default|GetShootSpreadMultiplier)) (Variables|Default|GetTossSpreadMultiplier)))
  (Variables|Default|SetShootFuseDelay (* (ScalableFloat|GetValueatLevel (Variables|Default|GetShootDelayMultiplier)) (Variables|Default|GetShootFuseDelay)))
  (Variables|Default|SetTossVelocity (* (ScalableFloat|GetValueatLevel (Variables|Default|GetShootVelocityMultiplier)) (Variables|Default|GetTossVelocity)))
  (if (not (Variables|Default|GetbPlayerInteractable))
    (Variables|BuildingActor|SetbIsInvulnerable true :!pinned true) ;  not player interactable chest, make invulnerable
    (Audio|Components|Audio|Stop (Variables|Default|GetAmbientLoop))
    (Rendering|SetVisibility (Variables|Default|Getchestglow) false false)))

(event Custom|SpawnDestroyFX
  ;[1 1 1] OnDestroyed
  (Niagara|SpawnSystematLocation (Variables|Default|GetCannonDestroyedEffect) (Transformation|GetActorLocation) (Transformation|GetActorRotation) "1.000000,1.000000,1.000000" true true "AutoRelease"))

(event EventOnDeathServer
  (params (Damage :type Float%20%28single-precision%29) (DamageTags :type Gameplay%20Tag%20Container%20Structure%20%28by%20ref%29) (Momentum :type Vector :default "0, 0, 0") (HitInfo :type Hit%20Result%20Structure%20%28by%20ref%29) (InstigatedBy :type Controller%20Object%20Reference) (DamageCauser :type Actor%20Object%20Reference) (EffectContext :type Gameplay%20Effect%20Context%20Handle%20Structure))
  ;[1 1 1] OnDestroyed
  (Networking|SwitchHasAuthority
    (:Authority
      (if (not (Variables|Default|GetbFiredLoot))
        (Inventory|SpawnPickupInWorldWithLootTier :LootTierName (Variables|Default|GetLootTierToDropOnDestroy) :Position (Transformation|GetActorLocation) :OverrideMaxStackCount 0 :bToss true :bTossWithVelocity false :SourceType "Destruction" :Source "Unset" :LootWorldLevel 0)
        (Variables|Default|SetbFiredLoot true))
      (CallFunction|SpawnDestroyFX))))

(event Interaction|EventBlueprintOnInteract
  (params (InteractingPawn :type Fort%20Pawn%20Object%20Reference) (InteractionBeingAttempted :type EInteractionBeingAttempted%20Enum :default "FirstInteraction"))
  ;[1 1 1] PlayerInteract
  (bind _output_get (Variables|Default|SetInteracting%20Pawn InteractingPawn))
  (Variables|Default|SetbPlayerInteractable false)
  (Utilities|FlowControl|DoOnce false
    (:Completed
      (CallFunction|FireLootCannon (Variables|Default|GetLootTierToFire) _output_get))))

(event Custom|FireLootCannon
  (params (LootTierToFire :type Name) (InteractingPawn :type Fort%20Pawn%20Object%20Reference))
  ;[1 1 1] Fire the Cannon
  (Variables|Default|SetInteracting%20Pawn InteractingPawn)
  (bind _output_get (Variables|Default|SetLootTierToFire LootTierToFire))
  (if (Variables|Default|GetbPlayerInteractable)
    (Utilities|FlowControl|DoOnce false
      (:Completed
        _)) ; only allow once shot if player interactable
    (else
      (Variables|Default|SetbFiredLoot false)))
  (CallFunction|LightFuse)
  (Utilities|FlowControl|Delay (Variables|Default|GetShootFuseDelay))
  (if (not (Variables|Default|GetbFiredLoot))
    (CallFunction|LaunchPlayers)
    (Networking|SwitchHasAuthority
      (:Authority
        (if (not (ScalableFloat|GetValueasBool (Variables|Default|GetDisableLootLaunch)))
          (Variables|Default|SetbFiredLoot true)
          (bind _returnvalue (Inventory|SpawnPickupInWorldWithLootTier (Variables|Default|GetLootTierToFire) (Transformation|GetWorldLocation (Variables|Default|GetArrow)) 0 0 true true (* (Transformation|GetForwardVector (Variables|Default|GetArrow)) (Math|Conversions|ToVector(Float) (Variables|Default|GetTossVelocity))) "Tossed" "ItemSpawner" (Loot|GetLootLevel)))
          (Variables|Default|SetSpawnedPickupLoot _returnvalue)
          (bind _interacting_pawn (Variables|Default|GetInteracting%20Pawn)
            (:then
              (CustomVerb|BroadcastCustomVerbMessage (CustomVerb|MakeCustomVerbMessage _interacting_pawn "(TagName=\"CustomVerb.ThemePark.LootCannon.Used\")")))
            (:else
              _))
          (for _array_element (Variables|Default|GetSpawnedPickupLoot)
            (bind _asfort_pickup_athena (Utilities|Casting|CastToFortPickupAthena _array_element))
            (Pickup|TossPickupinDirection _asfort_pickup_athena (+ (Math|Vector|MakeVector (* (Variables|Default|GetTossSpreadMultiplier) (Math|Random|RandomFloatinRange -4e+02 4e+02)) (* (Math|Random|RandomFloatinRange -4e+02 4e+02) (Variables|Default|GetTossSpreadMultiplier)) (* (Math|Random|RandomFloatinRange 1e+02 2.5e+02) (Variables|Default|GetTossSpreadMultiplier))) (* (Transformation|GetForwardVector (Variables|Default|GetArrow)) (Variables|Default|GetTossVelocity))) self false false))
          (Utilities|FlowControl|Delay (Variables|Default|GetLootSpawnAudioDelay))
          (Variables|Default|SetSingleUseLootCannonFired true)))
      (:Remote
        (CallFunction|CameraShake)
        (Components|Activation|Activate (Variables|Default|GetCannonFireEffect) false)
        (Audio|Components|Audio|Play (Variables|Default|GetLootLaunchAudio) 0.0)
        (Utilities|FlowControl|Delay (Variables|Default|GetLootSpawnAudioDelay))
        (for _array_element_2 (Variables|Default|GetSpawnedPickupLoot)
          (bind _asfort_pickup_athena_2 (Utilities|Casting|CastToFortPickupAthena _array_element_2))
          (Utilities|Array|Add (Variables|Default|GetLootLocation) (Transformation|GetWorldLocation (Variables|/Script/Engine.Actor|GetRootComponent _asfort_pickup_athena_2))))
        (Audio|SpawnSoundatLocation "/LootCannon/Audio/SC_LootCannon_Loot_Spawn.SC_LootCannon_Loot_Spawn" (Math|Vector|GetVectorArrayAverage (Variables|Default|GetLootLocation)))))))

(event Custom|LaunchPlayers
  ;[1 1 1] Launch Player
  (if (Variables|Default|GetbLaunchPlayerEnabled)
    (for _array_element (Collision|GetOverlappingActors (Variables|Default|GetFrontBox) "/Script/FortniteGame.FortPlayerPawnAthena" :!class /Script/Engine.PrimitiveComponent)
      (bind _asfort_player_pawn_athena (Utilities|Casting|CastToFortPlayerPawnAthena _array_element))
      (if (not (Health|IsDBNO _asfort_player_pawn_athena))
        (Character|LaunchCharacter _asfort_player_pawn_athena (* (Transformation|GetForwardVector (Variables|Default|GetArrow)) (Math|Conversions|ToVector(Float) (Math|Float|SafeDivide (Variables|Default|GetTossVelocity) 2.0))) true true)))))

(event Custom|CameraShake
  ;[1 1 1] Camera Shake
  (Networking|SwitchHasAuthority
    (:Remote
      (bind _outactors (Collision|SphereOverlapActors (Transformation|GetWorldLocation (Variables|Default|GetStaticMesh)) 7.5e+02 (Utilities|Array|MakeArray "ObjectTypeQuery3") "/Script/FortniteGame.FortPlayerPawnAthena"))
      (for _array_element _outactors
        (bind _asfort_player_controller_athena (Utilities|Casting|CastToFortPlayerControllerAthena (Fortnite|GetControllerfromActor _array_element)))
        (Game|Feedback|ClientStartCameraShake :self _asfort_player_controller_athena :Shake "/Game/Effects/Fort_Effects/Camera_Effects/BP_CameraShake_Cannon_Fire.BP_CameraShake_Cannon_Fire_C" :Scale 1.0 :PlaySpace "World")))))

(event Custom|LightFuse
  ;[1 1 1] Fuse Cosmetics
  (Networking|SwitchHasAuthority
    (:Remote
      (Components|Activation|Activate (Variables|Default|GetFuseEffect) false)
      (Audio|Components|Audio|Play (Variables|Default|GetFuseAudioCue) 0.0)
      (Utilities|FlowControl|Delay (Variables|Default|GetShootFuseDelay))
      (Components|Activation|Deactivate (Variables|Default|GetFuseEffect))
      (Audio|Components|Audio|FadeOut (Variables|Default|GetFuseAudioCue) 0.25 0.0 "Linear" :!class /Script/Engine.AudioComponent))))

(fn BlueprintCanInteract
  (params (InteractingPawn :type Fort%20Pawn%20Object%20Reference) (InteractionBeingAttempted :type EInteractionBeingAttempted%20Enum :default "FirstInteraction") (InteractionType :type TInteractionType%20Enum :default "IT_NoInteraction"))
  (return (Variables|Default|GetbPlayerInteractable)))

(fn BlueprintGetInteractionString
  (params (InteractingPawn :type Fort%20Pawn%20Object%20Reference) (InteractionBeingAttempted :type EInteractionBeingAttempted%20Enum :default "FirstInteraction"))
  (return "NSLOCTEXT(\"[29787D8E32EA9F30EEF318BCC2073DA1]\", \"F38192CC4DC8DCBB4A67BB864B3441CF\", \"Light Fuse\")"))

(fn OnRep_SingleUseLootCannonFired
  (bind _ambientloop (Variables|Default|GetAmbientLoop))
  (if (Variables|Default|GetSingleUseLootCannonFired)
    (Components|Activation|Deactivate (Variables|Default|Getchestglow))
    (if (Audio|Components|Audio|IsPlaying _ambientloop)
      (Audio|Components|Audio|FadeOut _ambientloop 1.0 0.0 "Linear" :!class /Script/Engine.AudioComponent)
      (else
        _))
    (Rendering|SetVisibility (Variables|Default|GetPointLight) false false)))

(fn ConstructionScript
  (bind _chestglow (Variables|Default|Getchestglow))
  (if (Variables|Default|GetbPlayerInteractable)
    (Rendering|SetVisibility _chestglow true false)
    (else
      (Rendering|SetVisibility _chestglow false false))))
